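# Masked parent material for the Brutal Burglar gadget prop, layering an animatable gold transition onto the 2023 character parent functions.
# A GoldTransition_Amount-driven sphere mask in pre-skinned local space, broken up by tiled noise and rounded, defines the gold region.
# A slightly smaller sphere is subtracted to form a glowing orange edge band.
# GoldTransition_HideBoneGeometry optionally turns the transition into an opacity cutout, revealing geometry as it spreads.
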Material /BRCosmetics/Animation/Props/Gadgets_BrutalBurglar/Materials/M_Gadget_BrutalBurglar_Parent
graph {
 properties {
  bUseMaterialAttributes=true
  bUsedWithMeshDeformer=true
  bUsedWithMorphTargets=true
  bUsedWithSkeletalMesh=true
  blendMode=BLEND_Masked
  pixelDepthOffsetMode=PDOM_Legacy
  shadingModel=MSM_FromMaterialExpression
  usedShadingModels=Default Lit | Subsurface | Clear Coat | Subsurface Profile | Cloth
 }
 nodes [
  Frac(None=Multiply(A=Time()) {constB=0.2})
  n1 = MaterialFunctionCall() {materialFunction=&MF_CharacterParent_2023_PreFX}
  MaterialFunctionCall(
   Material=NamedRerouteUsage()
   M=@n1.M
   Pre Skinned Local Position=NamedRerouteDeclaration(None=@n1.Pre skinned Local Position) {name=PreskinnedPosition, nodeColor=<rgba 0.1719 0 1 1>}
   Distance Blend=@n1.M
   Fresnel=@n1.M
  ) {materialFunction=&MF_CharacterParent_2023_PostFX}
  NamedRerouteDeclaration(None=@n1.Tech Art Mask) {name=TechArtMask, nodeColor=<rgba 0 0.6719 1 1>}
  n2 = NamedRerouteUsage()
  n3 = ScalarParameter() {defaultValue=0.15, parameterName=GoldSpecular_Offset}
  n4 = NamedRerouteUsage()
  n5 = SetMaterialAttributes!(
   MaterialAttributes=@n2
   Base Color=Constant3Vector() {constant=<rgba 0.7231 0.4564 0.1248 1>}
   Metallic=Constant() {r=1}
   Specular=Saturate(None=Add(A=@n3, B=@n4))
   Roughness=ScalarParameter() {defaultValue=0.3, group=BrutalBurglar, parameterName=GoldRoughness}
   Emissive Color
  ) {attributeSetTypes [
    #1
    #2
    #3
    #4
    #5
   ]}
  n6 = NamedRerouteUsage()
  n7 = NamedRerouteUsage()
  n8 = Subtract(A=OneMinus(None=ComponentMask(None=@n7) {b=true})) {constB=0.95}
  n9 = Multiply(A=@n6, B=Saturate(None=Multiply(A=@n8) {constB=20}))
  n10 = Add(A=@n6, B=OneMinus(None=ScalarParameter() {parameterName=GoldTransition_HideBoneGeometry}))
  n11 = Multiply(A=NamedRerouteUsage(), B=Constant3Vector() {constant=<rgba 1 0.4166 0 1>}) {constB=100}
  NamedRerouteDeclaration(None=SetMaterialAttributes!(
    MaterialAttributes=BlendMaterialAttributes(A=@n2, B=@n5, Alpha=@n9)
    Opacity Mask=Saturate(None=@n10)
    Emissive Color=Multiply(A=@n11) {constB=10}
   ) {attributeSetTypes [#6, #5]}) {name=OutMA, nodeColor=<rgba 0 1 0.3672 1>}
  n12 = GetMaterialAttributes(None=NamedRerouteDeclaration(None=@n1.Result) {name=InMA, nodeColor=<rgba 0.9922 1 0 1>}) {attributeGetTypes [#3, #4]}
  NamedRerouteDeclaration(None=@n12.Specular) {name=InSpecular, nodeColor=<rgba 0 0.2031 1 1>}
  NamedRerouteDeclaration(None=@n12.Roughness) {name=InRoughness, nodeColor=<rgba 0.5234 1 0 1>}
  n13 = NamedRerouteUsage()
  n14 = VectorParameter() {group=SuperLevelling_Mask, parameterName=SL_MaskPos_01, defaultValue=<rgba 10 0 0 1>}
  n15 = LinearInterpolate(Alpha=ScalarParameter() {group=BrutalBurglar, parameterName=GoldTransition_Amount}) {constB=82}
  n16 = ScalarParameter() {defaultValue=0.5, group=SuperLevelling_Mask, parameterName=SL_CoverMaskHardness}
  n17 = SphereMask(A=@n13, B=@n14.RGB, Radius=@n15, Hardness=@n16)
  n18 = TextureObjectParameter() {
   group=SuperLevelling_Mask
   parameterName=SL_TransitionNoise
   texture=/Game/Characters/Player/Male/Medium/Bodies/M_MED_Viper/Textures/Glitch/T_SoftNoise
  }
  n19 = VectorParameter() {group=SuperLevelling_Mask, parameterName=SL_TransitionNoise_Tiling, defaultValue=<rgba 24 10 10 1>}
  n20 = Multiply(
   A=ChannelMaskParameter(None=MaterialFunctionCall(TextureObject=@n18, TextureSize=@n19.RGB, WorldPosition=@n13) {materialFunction=/Game/Weapons/WeaponSkins/Wraps/MaterialFunctions/WorldAlignedTexture}.XYZ Texture) {group=SuperLevelling_Mask, parameterName=SL_TransitionNoiseChannel}.R
   B=LinearInterpolate(
    A=ScalarParameter() {defaultValue=0.5, group=SuperLevelling_Mask, parameterName=SL_Transition_Falloff_Start}
    B=ScalarParameter() {defaultValue=0.1, group=SuperLevelling_Mask, parameterName=SL_Transition_Falloff_End}
   )
  )
  n21 = Saturate(None=Round(None=Subtract(A=@n17, B=@n20)))
  NamedRerouteDeclaration(None=@n21) {name=GoldTransitionMask, nodeColor=<rgba 1 0 0.2109 1>}
  n22 = SphereMask(A=@n13, B=@n14.RGB, Radius=Max(A=Add(A=@n15) {constB=-1.5}) {constB=0}, Hardness=@n16)
  n23 = Saturate(None=Round(None=Subtract(A=@n22, B=@n20)))
  NamedRerouteDeclaration(None=Saturate(None=Subtract(A=@n21, B=@n23))) {name=EmissiveMask, nodeColor=<rgba 0 0.3906 1 1>}
 ]
}
